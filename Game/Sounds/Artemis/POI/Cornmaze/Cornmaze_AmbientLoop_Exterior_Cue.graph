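# Looping exterior ambience Sound Cue for the Cornmaze point of interest.
# A mixer blends three looping wave players at different levels: two corn stalk layers, with the second dominant, and a "Dizzy" layer.
# It plays through the outdoor ambient sound class with a shared ambient attenuation asset and a slightly reduced overall volume.
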SoundCue /Game/Sounds/Artemis/POI/Cornmaze/Cornmaze_AmbientLoop_Exterior_Cue
root=Mixer_1
nodes [
 {id=Mixer_1, children [WavePlayer_2,WavePlayer_3,WavePlayer_6], properties {inputVolume [0.3,1,0.5]}}
 {id=WavePlayer_2, properties {bLooping=true, soundWaveAssetPtr=&/Cornmaze_AmbientLoop_CornStalks_01}}
 {id=WavePlayer_3, properties {bLooping=true, soundWaveAssetPtr=&/Cornmaze_AmbientLoop_CornStalks_03}}
 {id=WavePlayer_6, properties {bLooping=true, soundWaveAssetPtr=&/Cornmaze_AmbientLoop_Dizzy_01}}
]
properties {
 attenuationSettings=&Fort_Ambient_Athena_Att
 maxDistance=2097152
 soundClassObject=&Fort_Ambient_Outdoors_Athena
 volumeMultiplier=0.852
 preEffectBusSends [{}]
}
