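# Sound cue for the daytime interior ambient bed of the non-haunted cornmaze point of interest.
# A mixer blends two looping corn stalk rustle waves with an alpine mountain day stereo ambience, whose player is not set to loop; one mixer input is left empty.
# The cue plays slightly pitched down at reduced volume, using the ambient attenuation asset and the indoor ambient sound class.
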SoundCue /Game/Sounds/Artemis/POI/Cornmaze/Cornmaze_AmbientLoop_NonHaunted_Interior_Day_Cue
root=Mixer_1
nodes [
 {id=Mixer_1, children [WavePlayer_2,WavePlayer_3,null,WavePlayer_4], properties {inputVolume [0.6,1,0.5,0.6]}}
 {id=WavePlayer_2, properties {bLooping=true, soundWaveAssetPtr=&/Cornmaze_AmbientLoop_CornStalks_01}}
 {id=WavePlayer_3, properties {bLooping=true, soundWaveAssetPtr=&/Cornmaze_AmbientLoop_CornStalks_03}}
 {id=WavePlayer_4, properties {soundWaveAssetPtr=/Game/Sounds/Artemis/Biomes/Mountains/Ambience_Mountains_Alpine_Day_Stereo}}
]
properties {
 attenuationSettings=&Fort_Ambient_Athena_Att
 maxDistance=2097152
 pitchMultiplier=0.9
 soundClassObject=&Fort_Ambient_Indoors_Athena
 volumeMultiplier=0.376
 preEffectBusSends [{}]
}
